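# Physics asset for the Flora Brisk backpack cosmetic, defining collision and joints for two dangling chains on the pack, one long and one short.
# A large capsule covers the pack, and a small base body for each chain is attached to it with all angular motion locked.
# Three capsule links hang from each base through limited swing and tight twist constraints, ending in a fatter capsule at each tip.
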PhysicsAsset /BRCosmetics/Accessories/FORT_Backpacks/Backpack_F_MED_FloraBrisk/Meshes/F_MED_FloraBrisk_Pack_Physics
bodies {
 base_dyn_longChain [{shapeType=Capsule, length=0.505, radius=0.505, shapeName=base_dyn_longChain_capsule, rotation=<rot 0 -90 0>}]
 base_dyn_shortChain [{shapeType=Capsule, length=0.505, radius=0.505, shapeName=base_dyn_shortChain_capsule, rotation=<rot 0 -90 0>}]
 dyn_longChain_00 [{shapeType=Capsule, length=2.505, radius=0.505, shapeName=dyn_longChain_00_capsule, center=<1.6332 0 0>, rotation=<rot 0 -89.9995 -90>}]
 dyn_longChain_01 [{
   shapeType=Capsule
   length=2.505
   radius=0.505
   shapeName=dyn_longChain_01_capsule
   center=<1.2176 -0.0232 -0.0137>
   rotation=<rot 0 -89.9445 -89.9978>
  }]
 dyn_longChain_02 [{
   shapeType=Capsule
   length=0.1268
   radius=2.3
   shapeName=dyn_longChain_02_capsule
   center=<1.3626 0.3379 0.1582>
   rotation=<rot 0 -106.7035 -55.1553>
  }]
 dyn_shortChain_00 [{
   shapeType=Capsule
   length=2.005
   radius=0.505
   shapeName=dyn_shortChain_00_capsule
   center=<1.0079 0.0592 -0.0222>
   rotation=<rot 84.705 89.8742 -90.0012>
  }]
 dyn_shortChain_01 [{
   shapeType=Capsule
   length=1.505
   radius=0.505
   shapeName=dyn_shortChain_01_capsule
   center=<0.7486 -0.0002 0.0047>
   rotation=<rot 0 -90.3101 -89.6241>
  }]
 dyn_shortChain_02 [{
   shapeType=Capsule
   length=0.6269
   radius=1.95
   shapeName=dyn_shortChain_02_capsule
   center=<1.4316 0.2396 -0.0678>
   rotation=<rot 0 -44.1435 -85.7133>
  }]
 pack [{shapeType=Capsule, length=1.5, radius=7, shapeName=pack_capsule, center=<0.1121 -7.8687 -1.8514>, rotation=<rot 23.8734 -91.4168 -2.8184>}]
}
constraints [
 {bone1Name=dyn_longChain_01, bone2Name=dyn_longChain_00, swing1LimitDegrees=25, swing2LimitDegrees=25, twistLimitDegrees=5}
 {bone1Name=dyn_longChain_02, bone2Name=dyn_longChain_01, swing1LimitDegrees=25, swing2LimitDegrees=25, twistLimitDegrees=5}
 {bone1Name=dyn_shortChain_01, bone2Name=dyn_shortChain_00, swing1LimitDegrees=25, swing2LimitDegrees=25, twistLimitDegrees=5}
 {bone1Name=dyn_shortChain_02, bone2Name=dyn_shortChain_01, swing1LimitDegrees=25, swing2LimitDegrees=25, twistLimitDegrees=5}
 {bone1Name=dyn_shortChain_00, bone2Name=base_dyn_shortChain, swing1LimitDegrees=25, swing2LimitDegrees=25, twistLimitDegrees=5}
 {bone1Name=dyn_longChain_00, bone2Name=base_dyn_longChain, swing1LimitDegrees=25, swing2LimitDegrees=25, twistLimitDegrees=5}
 {
  bone1Name=base_dyn_shortChain
  bone2Name=pack
  swing1LimitDegrees=45
  swing1Motion=Locked
  swing2LimitDegrees=45
  swing2Motion=Locked
  twistLimitDegrees=45
  twistMotion=Locked
 }
 {
  bone1Name=base_dyn_longChain
  bone2Name=pack
  swing1LimitDegrees=45
  swing1Motion=Locked
  swing2LimitDegrees=45
  swing2Motion=Locked
  twistLimitDegrees=45
  twistMotion=Locked
 }
]
