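# Animation Blueprint for the FloraBrisk backpack cosmetic that adds physics-driven secondary motion to the pack's dynamic bones.
# The input pose runs through a RigidBody node based at the pack bone, then four chained spring-driven AnimDynamics nodes for top, bottom and both loop bones.
# The event graph updates a gravity override variable each frame and selects front-end or in-zone sim space settings on customization load.
# Neither variable is visibly wired into the physics nodes.
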AnimBlueprint /BRCosmetics/Accessories/FORT_Backpacks/Backpack_F_MED_FloraBrisk/Meshes/F_MED_FloraBrisk_Pack_AnimBP
target_skeleton=&/F_MED_FloraBrisk_Pack_Skeleton
graphs [
 AnimGraph {nodes [
   n1 = RigidBody!(
    ComponentPose=Local To Component(LocalPose=Input Pose())
    Alpha
    OverrideWorldGravity:Vector="0,0,-980"
    ExternalForce
    SimSpaceSettings:Sim Space Settings Structure=(WorldAlpha=0,VelocityScaleZ=1,DampingAlpha=1,MaxLinearVelocity=10000,MaxAngularVelocity=10000,MaxLinearAcceleration=10000,MaxAngularAcceleration=10000,ExternalLinearDragV=(X=0,Y=0,Z=0),ExternalLinearVelocity=(X=0,Y=0,Z=0),ExternalAngularVelocity=(X=0,Y=0,Z=0))
   ) -> (Pose) {
    bOverrideWorldGravity=true
    overridePhysicsAsset=&/F_MED_FloraBrisk_Pack_Physics
    baseBoneRef {boneName=pack}
    overrideWorldGravity {z=-980}
   }
   n2 = Anim Dynamics - Bone: dyn_top_00(ComponentPose=@n1, Alpha=1, ExternalForce="0,0,0") {
    angularDampingOverride=1
    angularSpringConstant=30000
    bAngularSpring=true
    bOverrideAngularDamping=true
    bUseGravityOverride=true
    gravityScale=0.35
    simulationSpace=RootRelative
    boundBone {boneName=dyn_top_00}
    componentLinearAccScale=<0.3 0.3 0.4>
    componentLinearVelScale=<0.3 0.3 0.4>
    gravityOverride {z=-980}
    physicsBodyDefinitions [{
      boundBone {boneName=dyn_top_00}
      boxExtents=<30 40 40>
      constraintSetup {angularLimitsMax=<1 7 15>, angularLimitsMin=<-1 -7 -15>, angularTarget=<1 0 0>}
      localJointOffset=<4 0 0>
     }]
   }
   n3 = Anim Dynamics - Bone: dyn_bottom_00(ComponentPose=@n2, Alpha=1, ExternalForce="0,0,0") {
    angularDampingOverride=1
    angularSpringConstant=9000
    bAngularSpring=true
    bOverrideAngularDamping=true
    bUseGravityOverride=true
    gravityScale=0.35
    simulationSpace=RootRelative
    boundBone {boneName=dyn_bottom_00}
    componentLinearAccScale=<0.3 0.3 0.4>
    componentLinearVelScale=<0.3 0.3 0.4>
    gravityOverride {z=-980}
    physicsBodyDefinitions [{
      boundBone {boneName=dyn_bottom_00}
      boxExtents=<40 50 50>
      constraintSetup {angularLimitsMax=<1 7 15>, angularLimitsMin=<-1 -7 -15>, angularTarget=<1 0 0>}
      localJointOffset=<4 0 0>
     }]
   }
   n4 = Anim Dynamics - Bone: dyn_leftLoop_00(ComponentPose=@n3, Alpha=1, ExternalForce="0,0,0") {
    angularDampingOverride=1
    angularSpringConstant=1000
    bAngularSpring=true
    bOverrideAngularDamping=true
    bUseGravityOverride=true
    gravityScale=0.35
    simulationSpace=RootRelative
    boundBone {boneName=dyn_leftLoop_00}
    componentLinearAccScale=<0.3 0.3 0.4>
    componentLinearVelScale=<0.3 0.3 0.4>
    gravityOverride {z=-980}
    physicsBodyDefinitions [{
      boundBone {boneName=dyn_leftLoop_00}
      boxExtents=<20 30 30>
      constraintSetup {angularLimitsMax=<7 7 15>, angularLimitsMin=<-7 -7 -15>, angularTarget=<1 0 0>}
      localJointOffset=<4 0 0>
     }]
   }
   n5 = Anim Dynamics - Bone: dyn_rightLoop_00(ComponentPose=@n4, Alpha=1, ExternalForce="0,0,0") {
    angularDampingOverride=1
    angularSpringConstant=1000
    bAngularSpring=true
    bOverrideAngularDamping=true
    bUseGravityOverride=true
    gravityScale=0.35
    simulationSpace=RootRelative
    boundBone {boneName=dyn_rightLoop_00}
    componentLinearAccScale=<0.3 0.3 0.4>
    componentLinearVelScale=<0.3 0.3 0.4>
    gravityOverride {z=-980}
    physicsBodyDefinitions [{
      boundBone {boneName=dyn_rightLoop_00}
      boxExtents=<20 30 30>
      constraintSetup {angularLimitsMax=<15 15 15>, angularLimitsMin=<-15 -15 -15>, angularTarget=<1 0 0>}
      localJointOffset=<4 0 0>
     }]
   }
   Output Pose(Result=Component To Local(ComponentPose=@n5)) {layerGroup=None}
  ]}
 EventGraph {nodes [
   n1 = Event BlueprintUpdateAnimation()
   > n2 = "Gravity Override State Function (Parameter Asset) (Nativized)"(
    CharacterPartAnimInstance=Self-Reference()
    GameTimeInSeconds=@n1.DeltaTimeX
    Parameters=&/F_MED_FloraBrisk_Pack_Parameters
   )
   > Set Gravity Override(Gravity Override=@n2)
   Event OnCharacterCustomizationLoaded()
   > n3 = Branch(Condition=Get bIsInFrontEnd())
   > Set Sim Space Settings(Sim Space Settings:Sim Space Settings Structure=Get Sim Space Settings_FrontEnd() -> (Sim Space Settings_FrontEnd:Sim Space Settings Structure)) -> (Output_Get:Sim Space Settings Structure)
   Set Sim Space Settings(execute=@n3.else, Sim Space Settings=Get Sim Space Settings_InZone() -> (Sim Space Settings_InZone:Sim Space Settings Structure))
   TryGetPawnOwner()
  ]}
]
variables [
 Gravity Override {type=Vector}
 Sim Space Settings {
  type=SimSpaceSettings
  default {
   dampingAlpha=1
   maxAngularAcceleration=10000
   maxAngularVelocity=10000
   maxLinearAcceleration=10000
   maxLinearVelocity=10000
   velocityScaleZ=1
   worldAlpha=0
   externalLinearDragV=<0 0 0>
  }
 }
 Sim Space Settings_FrontEnd {
  type=SimSpaceSettings
  default {
   dampingAlpha=1
   maxAngularAcceleration=40
   maxAngularVelocity=8
   maxLinearAcceleration=4000
   maxLinearVelocity=800
   velocityScaleZ=1
   worldAlpha=0.4
   externalLinearDragV=<0.3 0.3 0>
  }
 }
 Sim Space Settings_InZone {
  type=SimSpaceSettings
  default {
   dampingAlpha=1
   maxAngularAcceleration=40
   maxAngularVelocity=8
   maxLinearAcceleration=4000
   maxLinearVelocity=800
   velocityScaleZ=1
   worldAlpha=0.65
   externalLinearDragV=<0.3 0.3 0>
  }
 }
]
